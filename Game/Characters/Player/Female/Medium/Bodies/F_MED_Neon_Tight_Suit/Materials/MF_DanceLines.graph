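# Material function for the Neon Tight Suit body that outputs clear-coat metallic attributes with animated, audio-reactive neon dance lines.
# Object-space sphere gradients, normal distortion and scrolling noise drive repeating bands sampled from a curve atlas and dimmed by view angle.
# A color dot-product mask isolates bands matching an alternating accent color for hue cycling and emissive boost.
# Emote audio overrides event spectrum, gating both effects and driving glow intensity.
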MaterialFunction /Game/Characters/Player/Female/Medium/Bodies/F_MED_Neon_Tight_Suit/Materials/MF_DanceLines
graph {nodes [
  n1 = CollectionParameter() {collection=&FortniteMaterialParameters, parameterName=EventAudio_Spectrum}
  n2 = MaterialFunctionCall(Float4=@n1) {materialFunction=&BreakOutFloat4Components}
  n3 = Time()
  n4 = Add(A=Sine(None=Multiply(A=@n3) {constB=0.5}))
  n5 = Power(Base=Multiply(A=LinearInterpolate(A=@n2.G, B=@n4) {constAlpha=0.7}) {constB=1.25})
  Clamp(None=@n5)
  n6 = Time()
  n7 = Multiply(A=@n2.R) {constB=1000}
  n8 = CollectionParameter() {collection=&FortniteMaterialParameters, parameterName=Emote_AudioLow}
  n9 = Clamp(None=Round(None=Multiply(A=@n8) {constB=200}))
  n10 = LinearInterpolate(A=Clamp(None=Round(None=@n7)), B=@n9, Alpha=@n9)
  n11 = Multiply(A=Frac(None=Multiply(A=@n6) {constB=0.1, desc=hue shift speed. was 0.25}), B=@n10)
  n12 = ScalarParameter() {defaultValue=1, group=Dance Lines, parameterName=StretchGradient}
  n13 = PreSkinnedPosition()
  n14 = Add(
   A=VectorParameter() {group=Dance Lines, parameterName=offset, defaultValue=<rgba 0 0 0 1>}.RGB
   B=FunctionInput(Preview=VertexInterpolator(VS=@n13).PS) {bUsePreviewValueAsDefault=true, inputName=Position}
  )
  n15 = SphereMask(
   A=MaterialFunctionCall(Float3=@n14) {materialFunction=&BreakOutFloat3Components}.G
   B=ScalarParameter() {defaultValue=150, desc=9.176, group=Dance Lines, parameterName=YAxisGradPosition}
   Radius=ScalarParameter() {defaultValue=50, desc=50, group=Dance Lines, parameterName=YAxisGradRadius}
   Hardness=Constant()
  )
  n16 = ScalarParameter() {defaultValue=1, group=Dance Lines, parameterName=YAxisGradIntensity}
  n17 = VectorParameter() {parameterName=ObjectSpaceGradStart, defaultValue=<rgba 0 40 150 1>}
  n18 = ScalarParameter() {defaultValue=180, group=Dance Lines, parameterName=ObjectSpaceGradRadius}
  n19 = Constant()
  n20 = GetMaterialAttributes(None=FunctionInput() {bUsePreviewValueAsDefault=true, inputName=Material, inputType=FunctionInput_MaterialAttributes}) {attributeGetTypes [#1]}
  n21 = Add(
   A=Add(A=Multiply(A=@n15, B=@n16), B=SphereMask(A=@n14, B=@n17.RGB, Radius=@n18, Hardness=@n19))
   B=Multiply(A=ComponentMask(None=@n20.Normal) {b=true}, B=ScalarParameter() {defaultValue=0.2, group=Dance Lines, parameterName=NormalDistortionWeight})
  )
  n22 = Multiply(A=Time(), B=ScalarParameter() {defaultValue=0.1, group=Dance Lines, parameterName=speed})
  n23 = Noise!(World Position=@n14, FilterWidth) {bTurbulence=false, levels=1, noiseFunction=NOISEFUNCTION_GradientALU, outputMin=0, scale=0.04}
  n24 = Add(
   A=Add(A=Multiply(A=@n21, B=@n12), B=@n22)
   B=Multiply(A=ComponentMask(None=@n23) {g=true}, B=ScalarParameter() {defaultValue=1, group=Dance Lines, parameterName=NoiseWarpIntensity})
  )
  n25 = ScalarParameter() {defaultValue=1.6, group=Dance Lines, parameterName=Repetition}
  n26 = CurveAtlasRowParameter(CurveTime=Frac(None=Multiply(A=@n24, B=@n25))) {
   atlas=&AthenaCurveAtlas01
   curve=&/NeonTightSuit_ColorCurve
   defaultValue=215
  }
  n27 = Normalize(VectorInput=@n26)
  n28 = Clamp(None=Frac(None=Multiply(A=@n22) {constB=2, desc=was 4 .. controlls which is the emissive accent . .A or B}))
  n29 = Normalize(VectorInput=LinearInterpolate(
    A=VectorParameter() {group=Dance Lines, parameterName=EmissiveAccentA, defaultValue=<rgba 0.9301 0.7991 0.0467 1>}.RGB
    B=VectorParameter() {group=Dance Lines, parameterName=EmissiveAccentB, defaultValue=<rgba 0.1804 0.8196 1 1>}.RGB
    Alpha=Round(None=@n28)
   ))
  n30 = SphereMask(
   A=Clamp(None=DotProduct(A=@n27, B=@n29))
   B=Constant() {r=1}
   Radius=ScalarParameter() {defaultValue=0.05, group=Dance Lines, parameterName=SphereRad}
   Hardness=Constant() {r=1}
  )
  n31 = DotProduct(A=PixelNormalWS(), B=CameraVectorWS())
  n32 = Multiply(
   A=Power(Base=Clamp(None=@n31), Exp=ScalarParameter() {defaultValue=2, parameterName=ViewDepedantFalloff})
   B=MaterialFunctionCall(Hue Shift Percentage=Multiply(A=@n11, B=@n30) {desc=hue shift only in certain bands}, Texture=@n26) {materialFunction=&HueShift}.Result
  )
  n33 = ScalarParameter() {defaultValue=0.3, parameterName=Emissive}
  n34 = LinearInterpolate(A=@n5, B=@n8, Alpha=@n9)
  n35 = LinearInterpolate(
   A=ScalarParameter() {defaultValue=0.3, group=Dance Lines, parameterName=EmissiveAccent}
   B=ScalarParameter() {defaultValue=3, group=Dance Lines, parameterName=Yellow Boost}
   Alpha=Multiply(A=@n30, B=Clamp(None=@n34))
  )
  n36 = SetMaterialAttributes!(
   MaterialAttributes
   ShadingModel=ShadingModel() {shadingModel=MSM_ClearCoat}
   BaseColor=Multiply(A=@n32, B=ScalarParameter() {defaultValue=1, parameterName=BaseColorBrightness})
   Metallic=ScalarParameter() {defaultValue=1, parameterName=Metalness}
   Specular
   Roughness=ScalarParameter() {defaultValue=0.5, parameterName=Roughness}
   Normal=MaterialFunctionCall(Normal=@n20.Normal, Flatness=ScalarParameter() {defaultValue=0.5, group=Dance Lines, parameterName=DanceLines_NormalFlattness}) {materialFunction=&FlattenNormal}.Result
   EmissiveColor=Multiply(A=@n32, B=LinearInterpolate(A=@n33, B=@n35, Alpha=@n10))
  ) {attributeSetTypes [
    #2
    #3
    #4
    #5
    #6
    #1
    #7
   ]}
  FunctionOutput(None=@n36)
  StaticBoolParameter() {parameterName=Remove Emote Audio Priority}
 ]}
